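# Material function for the Dual Paradox pickaxe that adds animated purple noise and fresnel glow to emissive, masking opacity to back faces only.
# Panning, turbulence-warped triplanar plasma, dotted with a time-driven hue shift, forms a phasing mask blending day or night color pairs.
# A masked fresnel, its exponent varying with view direction against instance axes, blends in glow color, optionally joined by depth fade for transparent use.
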MaterialFunction /Game/Weapons/FORT_Melee/Pickaxe_DualParadox/Materials/MFC_DualParadox_PurpleNoise
graph {nodes [
  CameraVectorWS()
  n1 = GetMaterialAttributes(None=FunctionInput(Preview=SetMaterialAttributes!(MaterialAttributes)) {bUsePreviewValueAsDefault=true, inputName=MA, inputType=FunctionInput_MaterialAttributes}) {attributeGetTypes [#1]}
  n2 = ScalarParameter() {parameterName=DualParadox_NoiseMin}
  n3 = ScalarParameter() {defaultValue=0.3, parameterName=DualParadox_NoiseMax}
  n4 = TextureObject() {texture=/Game/Effects/Fort_Effects/Textures/Plasma/T_Plasma_03_M}
  n5 = ScalarParameter() {defaultValue=19, parameterName=DualParadox_NoiseTextureSize}
  n6 = Time()
  n7 = Multiply(
   A=Normalize(VectorInput=VectorParameter() {parameterName=DualParadox_PannerDirection, defaultValue=<rgba 0 0 -0.2 0>}.RGB)
   B=ScalarParameter() {defaultValue=3, parameterName=DualParadox_PannerSpeed}
  )
  n8 = ComponentMask(None=MaterialFunctionCall(
    TextureObject=TextureObject() {samplerType=SAMPLERTYPE_LinearColor, texture=/Game/Effects/Fort_Effects/Textures/Plasma/T_Turbulence_AdditiveWarp_Linear}
    TextureSize=ScalarParameter() {defaultValue=35, parameterName=DualParadox_TurbulenceTextureSize}
    WorldPosition=NamedRerouteUsage()
    World Space Normal=NamedRerouteUsage()
    RotYAxis=ScalarParameter() {parameterName=DualParadox_TurbulenceRotation}
   ) {materialFunction=/Game/Weapons/WeaponSkins/Wraps/MaterialFunctions/WorldAlignedTexture}.XYZ Texture) {g=true, r=true}
  n9 = Constant()
  n10 = Add(
   A=Add(A=NamedRerouteUsage(), B=Multiply(A=@n6, B=@n7))
   B=Multiply(A=AppendVector(A=@n8, B=@n9), B=ScalarParameter() {defaultValue=2, parameterName=DualParadox_Distortion})
  )
  n11 = NamedRerouteUsage()
  n12 = ScalarParameter() {parameterName=DualParadox_NoiseRotationZ}
  n13 = ScalarParameter() {parameterName=DualParadox_NoiseRotationX}
  n14 = ScalarParameter() {parameterName=DualParadox_NoiseRotation}
  n15 = ScalarParameter() {parameterName=DualParadox_TriProjTransitionContrast}
  n16 = Multiply(A=Time(), B=ScalarParameter() {defaultValue=0.2, parameterName=DualParadox_NoiseDotSpeed})
  n17 = NamedRerouteDeclaration(None=DotProduct(
    A=MaterialFunctionCall(
     TextureObject=@n4
     TextureSize=@n5
     WorldPosition=@n10
     World Space Normal=@n11
     RotZAxis=@n12
     RotXAxis=@n13
     RotYAxis=@n14
     ProjectionTransitionContrast=@n15
    ) {materialFunction=&WorldAlignedTexture}.XYZ Texture
    B=MaterialFunctionCall(Hue Shift Percentage=@n16) {materialFunction=&HueShift}.Result
   )) {name=NoisePhasing, nodeColor=<rgba 0 0.7422 1 1>}
  n18 = FunctionOutput(None=Saturate(None=InverseLinearInterpolate(A=@n2, B=@n3, Value=@n17))) {outputName=Mask, sortPriority=30}
  n19 = LinearInterpolate(
   A=VectorParameter() {parameterName=DualParadox_Color2, defaultValue=<rgba 0.5149 0.117 0.7991 1>}.RGB
   B=VectorParameter() {parameterName=DualParadox_Color1, defaultValue=<rgba 0.7379 0.2195 1 1>}.RGB
   Alpha=@n18
  )
  n20 = LinearInterpolate(
   A=VectorParameter() {parameterName=DualParadox_Color2_Night, defaultValue=<rgba 0.6451 0.0406 0.7 1>}.RGB
   B=VectorParameter() {parameterName=DualParadox_Color1_Night, defaultValue=<rgba 0.9 0.1629 0.6174 1>}.RGB
   Alpha=@n18
  )
  n21 = ScalarParameter() {defaultValue=3, parameterName=DualParadox_FresnelExponentDown}
  n22 = Normalize(VectorInput=VectorParameter() {parameterName=DualParadox_FresnelForwardAxis, defaultValue=<rgba 0 1 0 1>}.RGB)
  n23 = Transform(None=Normalize(VectorInput=VectorParameter() {parameterName=DualParadox_FresnelUpAxis, defaultValue=<rgba 0 0 1 1>}.RGB)) {transformSourceType=TRANSFORMSOURCE_Instance}
  n24 = DotProduct(
   A=CrossProduct(A=Transform(None=@n22) {transformSourceType=TRANSFORMSOURCE_Instance}, B=@n23)
   B=MaterialFunctionCall() {materialFunction=&CameraDirectionVector}.Result
  )
  n25 = LinearInterpolate(
   A=ScalarParameter() {defaultValue=3, parameterName=DualParadox_FresnelExponent}
   B=ScalarParameter() {defaultValue=3, parameterName=DualParadox_FresnelExponentSide}
   Alpha=Power(Base=Abs(None=@n24), Exp=ScalarParameter() {defaultValue=2, parameterName=DualParadox_FresnelSideBlendPower})
  )
  n26 = ScalarParameter() {defaultValue=3, parameterName=DualParadox_FresnelExponentUp}
  n27 = DotProduct(A=MaterialFunctionCall() {materialFunction=&CameraDirectionVector}.Result, B=@n23)
  n28 = Abs(None=@n27)
  n29 = ScalarParameter() {defaultValue=2, parameterName=DualParadox_FresnelVerticalBlendPower}
  n30 = ConstantBiasScale(None=Multiply(A=Power(Base=@n28, Exp=@n29), B=Sign(None=@n27)))
  n31 = TwoSidedSign()
  n32 = FunctionInput(Preview=Transform(None=Constant3Vector() {constant=<rgba 0 0 1 0>})) {bUsePreviewValueAsDefault=true, inputName=Fresnel Normal, sortPriority=100}
  n33 = Multiply(
   A=Fresnel(
    ExponentIn=MaterialFunctionCall!(1=@n21, 2=@n25, 3=@n26, 4, A=@n30) {materialFunction=/Engine/Functions/Engine_MaterialFunctions02/Lerp_Multiple_Float}.Lerp 3 Inputs
    Normal=Multiply(A=@n31, B=@n32)
   ) {baseReflectFraction=0}
   B=ScalarParameter() {defaultValue=1, parameterName=FresnelGlowIntensity}
  )
  n34 = TextureSampleParameter2D() {
   parameterName=DualParadox_FresnelMask
   samplerType=SAMPLERTYPE_Masks
   texture=&T_BlackMask
   channelNames {a="", b="", g=No Glow, r=Permanent Glow}
  }
  n35 = OneMinus(None=Power(Base=@n34.No Glow, Exp=ScalarParameter() {defaultValue=0.1, parameterName=DualParadox_FresnelMaskExp}))
  n36 = ScalarParameter() {defaultValue=1, parameterName=DualParadox_FresnelPermkExp}
  n37 = FunctionOutput(None=Max(A=Multiply(A=@n33, B=@n35), B=Power(Base=@n34.Permanent Glow, Exp=@n36))) {outputName=Fresnel, sortPriority=60}
  n38 = NamedRerouteUsage()
  n39 = ScalarParameter() {defaultValue=0.25, parameterName=DualParadox_BaseGlow}
  n40 = Max(A=Max(A=@n37, B=Multiply(A=@n38, B=@n39)), B=NamedRerouteUsage())
  n41 = FunctionOutput(None=LinearInterpolate(
    A=MaterialFunctionCall(Color_Morning=@n19, Color_Day=@n19, Color_Evening=@n19, Color_Night=@n20) {materialFunction=&Color_Time_of_Day}.Result
    B=VectorParameter() {parameterName=DualParadox_FresnelColor, defaultValue=<rgba 7.5039 5.6061 10 1>}.RGB
    Alpha=StaticSwitchParameter(True=@n40, False=@n37) {parameterName="EnableDepthFade(TransparentOnly)"}
   )) {outputName=Color, sortPriority=5}
  n42 = Saturate(None=TwoSidedSign())
  FunctionOutput(None=SetMaterialAttributes!(
    MaterialAttributes=@n1.MaterialAttributes
    EmissiveColor=Add(A=@n1.EmissiveColor, B=@n41)
    Opacity=ScalarParameter() {defaultValue=0.5, parameterName=DualParadox_TransparentOpacity}
    OpacityMask=OneMinus(None=@n42)
   ) {attributeSetTypes [#1, #2, #3]})
  n43 = PreSkinnedPosition()
  NamedRerouteDeclaration(None=FunctionInput(Preview=VertexInterpolator(VS=@n43).PS) {bUsePreviewValueAsDefault=true, inputName=Position, sortPriority=20}) {name=Position, nodeColor=<rgba 0 0.625 1 1>}
  n44 = PreSkinnedNormal()
  NamedRerouteDeclaration(None=FunctionInput(Preview=VertexInterpolator(VS=@n44).PS) {bUsePreviewValueAsDefault=true, inputName=Normal, sortPriority=21}) {name=Normal, nodeColor=<rgba 1 0.0469 0 1>}
  n45 = Add(A=NamedRerouteUsage(), B=Constant3Vector() {constant=<rgba -9 0 0 0>})
  n46 = Normalize(VectorInput=Constant3Vector() {constant=<rgba 1 0.2 -0.15 0>})
  n47 = Divide(B=Distance(
    A=@n45
    B=MaterialFunctionCall!(Plane Normal=@n46, Vector=@n45) {materialFunction=/Engine/Functions/Engine_MaterialFunctions02/Math/ProjectVectorOntoPlane}.Result
   ))
  n48 = Multiply(A=@n47, B=ScalarParameter() {defaultValue=1.5, parameterName=DualParadox_DepthFadeDistance}) {constB=2}
  n49 = DepthFade(FadeDistance=LinearInterpolate(A=@n48, B=Multiply(A=@n48) {constB=2}, Alpha=NamedRerouteUsage()))
  n50 = ScalarParameter() {defaultValue=1, parameterName=DualParadox_DepthFadeExponent}
  NamedRerouteDeclaration(None=OneMinus(None=Power(Base=@n49, Exp=@n50))) {name=Depth Fade, nodeColor=<rgba 0 0.2031 1 1>}
  NamedRerouteDeclaration(None=Power(Base=@n47)) {name=Glow Gradient, nodeColor=<rgba 0 0.1094 1 1>}
  TextureSample(UVs=Multiply(A=TextureCoordinate(), B=ScalarParameter() {defaultValue=2, parameterName=DualParadox_TurbulenceNoiseScale})) {samplerType=SAMPLERTYPE_LinearColor, texture=&T_Turbulence_AdditiveWarp_Linear}
 ]}
